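# Control Rig for the female Agent X face-accessory part that applies pose-driven corrective offsets to hair bones.
# A nested sequence runs two spherical pose readers on the head bone, each aimed at a different rotation offset.
# Each reader's output weights a Modify Transforms node that rotates or shifts back and side hair base bones.
# A right-clavicle reader pair is unconnected to execution and targets a bone absent from the hierarchy.
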ControlRigBlueprint /BRCosmetics/Characters/Player/Female/Medium/Bodies/F_MED_Agent_X/Meshes/Parts/F_MED_Agent_X_FaceAcc_Ctrl_Rig
graphs [
 ContainedGraph {nodes [
   n1 = RigVMFunctionEntryNode()
   SequenceAggregate = &RigVMFunction_Sequence(ExecuteContext=@n1.ExecuteContext)
   SequenceAggregate_1 = &RigVMFunction_Sequence(ExecuteContext=@SequenceAggregate.B)
   RigVMFunctionReturnNode(A=@SequenceAggregate.A, B=@SequenceAggregate_1.A, C=@SequenceAggregate_1.B)
  ]}
 RigVMFunctionLibrary {links [], nodes []}
 RigVMModel {nodes [
   RigUnit_BeginExecution = &RigUnit_BeginExecution()
   SequenceAggregate_1 = RigVMAggregateNode(ExecuteContext=@RigUnit_BeginExecution.ExecutePin)
   n1 = &RigUnit_SphericalPoseReader(
    ActiveRegionScaleFactors=(PositiveWidth=1,NegativeWidth=1,PositiveHeight=1,NegativeHeight=1)
    ActiveRegionSize=0.1
    Debug=(bDrawDebug=True,bDraw2D=False,bDrawLocalAxes=False,DebugScale=25,DebugSegments=20,DebugThickness=0.25)
    DriverAxis=(X=1,Y=0,Z=0)
    DriverItem=(Type=Bone,Name="head")
    FalloffRegionScaleFactors=(PositiveWidth=1,NegativeWidth=1,PositiveHeight=1,NegativeHeight=1)
    FalloffSize=0.18
    FlipHeightScaling=false
    FlipWidthScaling=false
    OptionalParentItem=(Type=Bone)
    RotationOffset=(X=90,Y=0,Z=-125)
    ExecutePin=@SequenceAggregate_1.A
   )
   > &RigUnit_ModifyTransforms(
    ItemToModify="((Item=(Type=Bone,Name=\"dyn_hair_back_base_08\"),Transform=(Rotation=(X=0,Y=0,Z=-0.1478,W=0.989),Translation=(X=0,Y=0,Z=0),Scale3D=(X=1,Y=1,Z=1))),(Item=(Type=Bone,Name=\"dyn_hair_side_back_base\"),Transform=(Rotation=(X=0,Y=0,Z=-0.1478,W=0.989),Translation=( ...(truncated)"
    Weight=@n1.OutputParam
   )
   n2 = &RigUnit_SphericalPoseReader(
    ActiveRegionScaleFactors=(PositiveWidth=1,NegativeWidth=1,PositiveHeight=1,NegativeHeight=1)
    ActiveRegionSize=0.1
    Debug=(bDrawDebug=True,bDraw2D=False,bDrawLocalAxes=False,DebugScale=25,DebugSegments=20,DebugThickness=0.25)
    DriverAxis=(X=1,Y=0,Z=0)
    DriverItem=(Type=Bone,Name="head")
    FalloffRegionScaleFactors=(PositiveWidth=1,NegativeWidth=1,PositiveHeight=1,NegativeHeight=1)
    FalloffSize=0.18
    FlipHeightScaling=false
    FlipWidthScaling=false
    OptionalParentItem=(Type=Bone)
    RotationOffset=(X=90,Y=0,Z=-50)
    ExecutePin=@SequenceAggregate_1.C
   )
   > &RigUnit_ModifyTransforms(
    ItemToModify="((Item=(Type=Bone,Name=\"dyn_hair_back_base_10\"),Transform=(Rotation=(X=0,Y=0,Z=0,W=1),Translation=(X=2,Y=0,Z=0),Scale3D=(X=1,Y=1,Z=1))),(Item=(Type=Bone,Name=\"dyn_hair_side_back_base\"),Transform=(Rotation=(X=0,Y=0,Z=0,W=1),Translation=(X=2,Y=0,Z=0),Scale3D ...(truncated)"
    Weight=@n2.OutputParam
   )
   n3 = &RigUnit_SphericalPoseReader(
    ActiveRegionScaleFactors=(PositiveWidth=1,NegativeWidth=1,PositiveHeight=1,NegativeHeight=1)
    ActiveRegionSize=0.1
    Debug=(bDrawDebug=True,bDraw2D=False,bDrawLocalAxes=False,DebugScale=25,DebugSegments=20,DebugThickness=0.25)
    DriverAxis=(X=-1,Y=0,Z=0)
    DriverItem=(Type=Bone,Name="clavicle_r")
    ExecutePin=()
    FalloffRegionScaleFactors=(PositiveWidth=1,NegativeWidth=1,PositiveHeight=1,NegativeHeight=1)
    FalloffSize=0.18
    FlipHeightScaling=false
    FlipWidthScaling=false
    OptionalParentItem=(Type=Bone)
    RotationOffset=(X=0,Y=125,Z=0)
   )
   &RigUnit_ModifyTransforms(
    ExecutePin=()
    ItemToModify=((Item=(Type=Bone,Name="base_hair_back_a_00_r"),Transform=(Rotation=(X=0,Y=0,Z=0,W=1),Translation=(X=0,Y=0,Z=0),Scale3D=(X=1,Y=1,Z=1))))
    Weight=@n3.OutputParam
   )
  ]}
]
hierarchy [
 root {loc=<0 0 0>, rot=<0 0 0 1>}
 pelvis {parent=root, loc=<0 -0.5301 98.0571>, rot=<0.0127 0.707 -0.0127 -0.707>}
 spine_01 {parent=pelvis, loc=<6.1208 -0.9328 0>, rot=<0 0 0.034 -0.9994>}
 spine_02 {parent=spine_01, loc=<6.5996 -0.2095 0>, rot=<0 0 0.0238 -0.9997>}
 spine_03 {parent=spine_02, loc=<6.4914 -0.1237 0>, rot=<0 0 0.0149 -0.9999>}
 spine_04 {parent=spine_03, loc=<6.9598 0.0794 0>, rot=<0 0 -0.0132 -0.9999>}
 spine_05 {parent=spine_04, loc=<6.9987 0 0>, rot=<0 0 -0.0015 -1>}
 neck_01 {parent=spine_05, loc=<11.3138 0.1529 0>, rot=<0 0 -0.2003 -0.9797>}
 neck_02 {parent=neck_01, loc=<4.5627 0 0>, rot=<0 0 0.0504 -0.9987>}
 head {parent=neck_02, loc=<5.3914 0 0>, rot=<0 0 0.115 -0.9934>}
 dyn_hair_side_front_base {parent=head, loc=<1.7494 -0.9811 5.7017>, rot=<0.0842 -0.4516 0.888 -0.0211>}
 dyn_hair_side_front_parent {parent=dyn_hair_side_front_base, loc=<0 0 0>, rot=<0.0872 0 0 -0.9962>}
 dyn_hair_01 {parent=dyn_hair_side_front_parent, loc=<3.3852 -0.0261 -1.0637>, rot=<-0.0495 -0.1606 -0.0358 -0.9851>}
 dyn_hair_02 {parent=dyn_hair_01, loc=<2.5023 0 0>, rot=<0 0 0 1>}
 dyn_hair_03 {parent=dyn_hair_side_front_parent, loc=<3.7058 0 0>, rot=<-0.0872 0 0 -0.9962>}
 dyn_hair_side_mid_base {parent=head, loc=<2.7978 -2.5845 5.8752>, rot=<0.0356 -0.6651 0.7431 -0.0645>}
 dyn_hair_side_mid_parent {parent=dyn_hair_side_mid_base, loc=<0 0 0>, rot=<-0.0436 0 0 -0.999>}
 dyn_hair_04 {parent=dyn_hair_side_mid_parent, loc=<3.7997 0 0>, rot=<0.0872 0 0 -0.9962>}
 dyn_hair_side_back_base {parent=head, loc=<10.1607 -4.1456 4.7508>, rot=<-0.0225 -0.9363 0.3424 -0.0742>}
 dyn_hair_side_back_parent {parent=dyn_hair_side_back_base, loc=<4.4226 -1.4116 0.0572>, rot=<0 0 0 -1>}
 dyn_hair_05 {parent=dyn_hair_side_back_parent, loc=<4.6912 -0.4946 0>, rot=<0 0 0 -1>}
 dyn_hair_front_base {parent=head, loc=<2.5758 6.9283 -6.9879>, rot=<0.054 0.4807 0.8596 -0.1644>}
 dyn_hair_front_parent {parent=dyn_hair_front_base, loc=<0 0 0>, rot=<-0.0154 -0.1313 0.0753 -0.9884>}
 dyn_hair_18 {parent=dyn_hair_front_parent, loc=<3.2317 0 0>, rot=<0.1169 0.0416 -0.037 -0.9916>}
 dyn_hair_19 {parent=dyn_hair_18, loc=<3.163 0 0>, rot=<0 0 0 1>}
 dyn_hair_20 {parent=dyn_hair_front_parent, loc=<6.2328 0.9987 1.6806>, rot=<0.2707 0.1463 -0.0387 -0.9507>}
 dyn_hair_21 {parent=dyn_hair_20, loc=<2.896 -0.3415 0.5513>, rot=<-0.0408 0.0957 0.054 -0.9931>}
 dyn_hair_front_strand_base {parent=head, loc=<7.4341 10.1444 -3.223>, rot=<-0.0287 -0.9159 0.3694 -0.1545>}
 dyn_hair_front_strand_parent {parent=dyn_hair_front_strand_base, loc=<0 0 0>, rot=<0 0 0 1>}
 dyn_hair_front_strand_01 {parent=dyn_hair_front_strand_parent, loc=<3.3572 0.1124 0>, rot=<0 0 0 1>}
 dyn_hair_back_base_06 {parent=head, loc=<10.8203 -6.2242 2.4988>, rot=<-0.032 -0.981 0.1747 -0.0785>}
 dyn_hair_06 {parent=dyn_hair_back_base_06, loc=<4.4104 -1.5693 0.0159>, rot=<0 0 0 1>}
 dyn_hair_07 {parent=dyn_hair_06, loc=<5.8736 0 0.237>, rot=<0 0 0 1>}
 dyn_hair_back_base_08 {parent=head, loc=<11.0915 -7.1578 -2.1244>, rot=<0.0041 -0.9914 -0.1305 0.0005>}
 dyn_hair_08 {parent=dyn_hair_back_base_08, loc=<4.4763 -0.9468 0>, rot=<0.0079 0.0406 -0.0413 -0.9983>}
 dyn_hair_09 {parent=dyn_hair_08, loc=<5.7072 0 -0.3494>, rot=<0 0 0 1>}
 dyn_hair_back_base_10 {parent=head, loc=<11.4168 -5.9761 -5.1897>, rot=<-0.0023 -0.9345 -0.3557 -0.0146>}
 dyn_hair_10 {parent=dyn_hair_back_base_10, loc=<4.4737 -1.1082 -0.1539>, rot=<0 0 0 1>}
 dyn_hair_11 {parent=dyn_hair_10, loc=<5.9941 0 0>, rot=<0 0 0 1>}
 dyn_hair_back_base_12 {parent=head, loc=<11.6324 -2.9193 -6.6119>, rot=<-0.0519 -0.8888 -0.455 0.019>}
 dyn_hair_12 {parent=dyn_hair_back_base_12, loc=<4.4461 -2.4507 0.5392>, rot=<0 0 0 1>}
 dyn_hair_13 {parent=dyn_hair_12, loc=<5.9725 0 0>, rot=<0.1305 0 0 -0.9914>}
 dyn_hair_14 {parent=dyn_hair_13, loc=<3.8479 0.9986 0.2919>, rot=<-0.0805 0.0228 -0.1105 -0.9903>}
 dyn_hair_15_base {parent=head, loc=<10.6537 2.4766 -6.7427>, rot=<0.0029 -0.7071 -0.0029 -0.7071>}
 dyn_hair_15 {parent=dyn_hair_15_base, loc=<2.407 0 -4.4763>, rot=<0.5115 -0.5747 -0.4247 -0.4771>}
 dyn_hair_16 {parent=dyn_hair_15, loc=<6.788 0 0>, rot=<0.0067 0.0083 -0.1 -0.9949>}
 dyn_hair_17 {parent=dyn_hair_16, loc=<4.1907 0 0>, rot=<0 0 0 1>}
 clavicle_l {parent=spine_05, loc=<5.4831 -0.1053 -3.2946>, rot=<0.0465 -0.7289 -0.009 -0.6829>}
 upperarm_l {parent=clavicle_l, loc=<9.8658 -1.2136 -0.4105>, rot=<0.0208 -0.486 -0.0242 -0.8734>}
 clavicle_r {parent=spine_05, loc=<5.4831 -0.1053 3.2946>, rot=<0.7289 0.0465 -0.6829 0.009>}
 upperarm_r {parent=clavicle_r, loc=<-9.8658 1.2136 0.4104>, rot=<0.0208 -0.486 -0.0242 -0.8734>}
]
variables [
 Hair Middle L Transform {type=FTransform, default=(Rotation=(X=0,Y=-0.1908,Z=0,W=0.9816),Translation=(X=0,Y=0,Z=0),Scale3D=(X=1,Y=1,Z=1))}
 Hair Middle R Transform {type=FTransform, default=(Rotation=(X=-0.2541,Y=-0.1843,Z=-0.0494,W=0.9482),Translation=(X=0,Y=0,Z=0),Scale3D=(X=1,Y=1,Z=1))}
 Hair Out R Transform {type=FTransform, default=(Rotation=(X=-0.1913,Y=-0.4619,Z=0.1913,W=0.8446),Translation=(X=0,Y=0,Z=0),Scale3D=(X=1,Y=1,Z=1))}
 Hair Out R Transform_0 {type=FTransform, default=(Rotation=(X=-0.0782,Y=-0.0579,Z=0.1691,W=0.9808),Translation=(X=0,Y=0,Z=0),Scale3D=(X=1,Y=1,Z=1))}
 Hair Out R Transform_1 {type=FTransform, default=(Rotation=(X=0,Y=0,Z=0,W=1),Translation=(X=0,Y=1.25,Z=0),Scale3D=(X=1,Y=1,Z=1))}
 ItemToModify_0_Transform {type=FTransform, default=(Rotation=(X=0,Y=0,Z=0,W=1),Translation=(X=0,Y=0,Z=0),Scale3D=(X=1,Y=1,Z=1))}
 ItemToModify_0_Transform_Rotation {type=FQuat, default=(X=0,Y=0,Z=-0.5,W=0.866)}
]
