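# Physics asset for the Agent X female medium body's FaceAcc mesh part, setting up physics for the character's hair.
# Small capsules on dyn_hair bones form branching front, side and back chains, alongside larger capsules on head, neck_02 and upperarm_l.
# Each chain's base joins the head with free swing and twist.
# Constraints further down carry per-section angular limit values, tightest in the front sections and mostly wider at sides and back.
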PhysicsAsset /BRCosmetics/Characters/Player/Female/Medium/Bodies/F_MED_Agent_X/Meshes/Parts/F_MED_Agent_X_FaceAcc_Physics
bodies {
 dyn_hair_01 [{shapeType=Capsule, length=1, radius=1, shapeName=dyn_hair_01_capsule, center=<0.5 0 0>, rotation=<rot 90 0 0>}]
 dyn_hair_02 [{shapeType=Capsule, length=1, radius=1, shapeName=dyn_hair_02_capsule, center=<0.5 0 0>, rotation=<rot 90 0 0>}]
 dyn_hair_03 [{shapeType=Capsule, length=1, radius=1, shapeName=dyn_hair_03_capsule, center=<0.5 0 0>, rotation=<rot 90 0 0>}]
 dyn_hair_04 [{shapeType=Capsule, length=1, radius=1, shapeName=dyn_hair_04_capsule, center=<0.5 0 0>, rotation=<rot 90 0 0>}]
 dyn_hair_05 [{shapeType=Capsule, length=1, radius=1, shapeName=dyn_hair_05_capsule, center=<0.5 0 0>, rotation=<rot 90 0 0>}]
 dyn_hair_06 [{shapeType=Capsule, length=1, radius=1, shapeName=dyn_hair_06_capsule, center=<0.1 0 0>, rotation=<rot 90 0 0>}]
 dyn_hair_07 [{shapeType=Capsule, length=1, radius=1, shapeName=dyn_hair_07_capsule, center=<0.1 0 0>, rotation=<rot 90 0 0>}]
 dyn_hair_08 [{shapeType=Capsule, length=1, radius=1, shapeName=dyn_hair_08_capsule, center=<0.5 0 0>, rotation=<rot 90 0 0>}]
 dyn_hair_09 [{shapeType=Capsule, length=1, radius=1, shapeName=dyn_hair_09_capsule, center=<0.1 0 0>, rotation=<rot 90 0 0>}]
 dyn_hair_10 [{shapeType=Capsule, length=1, radius=1, shapeName=dyn_hair_10_capsule, center=<0.5 0 0>, rotation=<rot 90 0 0>}]
 dyn_hair_11 [{shapeType=Capsule, length=1, radius=1, shapeName=dyn_hair_11_capsule, center=<0.1 0 0>, rotation=<rot 90 0 0>}]
 dyn_hair_12 [{shapeType=Capsule, length=1, radius=1, shapeName=dyn_hair_12_capsule, center=<0.5 0 0>, rotation=<rot 90 0 0>}]
 dyn_hair_13 [{shapeType=Capsule, length=1, radius=1, shapeName=dyn_hair_13_capsule, center=<0.5 0 0>, rotation=<rot 90 0 0>}]
 dyn_hair_14 [{shapeType=Capsule, length=1, radius=1, shapeName=dyn_hair_14_capsule, center=<0.5 0 0>, rotation=<rot 90 0 0>}]
 dyn_hair_15 [{shapeType=Capsule, length=1, radius=1, shapeName=dyn_hair_15_capsule, center=<0.1 0 0>, rotation=<rot 90 0 0>}]
 dyn_hair_15_base [{shapeType=Capsule, length=1, radius=0.5, shapeName=dyn_hair_15_base_capsule, center=<0.1 0 0>}]
 dyn_hair_16 [{shapeType=Capsule, length=1, radius=1, shapeName=dyn_hair_16_capsule, center=<0.1 0 0>, rotation=<rot 90 0 0>}]
 dyn_hair_17 [{shapeType=Capsule, length=1, radius=1, shapeName=dyn_hair_17_capsule, center=<0.1 0 0>, rotation=<rot 90 0 0>}]
 dyn_hair_18 [{shapeType=Capsule, length=1, radius=1, shapeName=dyn_hair_18_capsule, center=<0.5 0 0>, rotation=<rot 90 0 0>}]
 dyn_hair_19 [{shapeType=Capsule, length=1, radius=1, shapeName=dyn_hair_19_capsule, center=<0.5 0 0>, rotation=<rot 90 0 0>}]
 dyn_hair_20 [{shapeType=Capsule, length=1, radius=1, shapeName=dyn_hair_20_capsule, center=<0.5 0 0>, rotation=<rot 90 0 0>}]
 dyn_hair_21 [{shapeType=Capsule, length=1, radius=1, shapeName=dyn_hair_21_capsule, center=<0.5 0 0>, rotation=<rot 90 0 0>}]
 dyn_hair_back_base_06 [{shapeType=Capsule, length=1, radius=0.5, shapeName=dyn_hair_back_base_06_capsule, center=<0.1 0 0>}]
 dyn_hair_back_base_08 [{shapeType=Capsule, length=1, radius=0.5, shapeName=dyn_hair_back_base_08_capsule, center=<0.1 0 0>}]
 dyn_hair_back_base_10 [{shapeType=Capsule, length=1, radius=0.5, shapeName=dyn_hair_back_base_10_capsule, center=<0.1 0 0>}]
 dyn_hair_back_base_12 [{shapeType=Capsule, length=1, radius=0.5, shapeName=dyn_hair_back_base_12_capsule, center=<0.1 0 0>}]
 dyn_hair_front_base [{shapeType=Capsule, length=0.5, radius=0.5, shapeName=dyn_hair_front_base_capsule}]
 dyn_hair_front_parent [{shapeType=Capsule, length=1, radius=1, shapeName=dyn_hair_front_parent_capsule, center=<0.1 0 0>, rotation=<rot 90 0 0>}]
 dyn_hair_front_strand_01 [{shapeType=Capsule, length=1, radius=1, shapeName=dyn_hair_front_strand_01_capsule, rotation=<rot 90 0 0>}]
 dyn_hair_front_strand_base [{shapeType=Capsule, length=0.5, radius=0.5, shapeName=dyn_hair_front_strand_base_capsule}]
 dyn_hair_front_strand_parent [{shapeType=Capsule, length=1, radius=1, shapeName=dyn_hair_front_strand_parent_capsule, rotation=<rot 90 0 0>}]
 dyn_hair_side_back_base [{shapeType=Capsule, length=1, radius=0.5, shapeName=dyn_hair_side_back_base_capsule, center=<0.1 0 0>}]
 dyn_hair_side_back_parent [{shapeType=Capsule, length=1, radius=1, shapeName=dyn_hair_side_back_parent_capsule, center=<0.1 0 0>, rotation=<rot 90 0 0>}]
 dyn_hair_side_front_base [{shapeType=Capsule, length=0.5, radius=0.5, shapeName=dyn_hair_side_front_base_capsule}]
 dyn_hair_side_front_parent [{shapeType=Capsule, length=1, radius=1, shapeName=dyn_hair_side_front_parent_capsule, center=<0.5 0 0>, rotation=<rot 90 0 0>}]
 dyn_hair_side_mid_base [{shapeType=Capsule, length=0.5, radius=0.5, shapeName=dyn_hair_side_mid_base_capsule}]
 dyn_hair_side_mid_parent [{shapeType=Capsule, length=1, radius=1, shapeName=dyn_hair_side_mid_parent_capsule, center=<0.5 0 0>, rotation=<rot 90 0 0>}]
 head [{shapeType=Capsule, length=6.1, radius=5.9, center=<0 -2 0>, rotation=<rot 90 0 0>}]
 neck_02 [{shapeType=Capsule, length=7, radius=6, rotation=<rot 90 0 0>}]
 upperarm_l [{shapeType=Capsule, length=5, radius=5, shapeName=upperarm_l_capsule, center=<-1.5 0 0.8>, rotation=<rot -30 0 0>}]
}
constraints [
 {
  bone1Name=dyn_hair_side_front_base
  bone2Name=head
  swing1LimitDegrees=45
  swing1Motion=Free
  swing2LimitDegrees=45
  swing2Motion=Free
  twistLimitDegrees=45
  twistMotion=Free
 }
 {bone1Name=dyn_hair_side_front_parent, bone2Name=dyn_hair_side_front_base, swing1LimitDegrees=45, swing2LimitDegrees=45, twistLimitDegrees=15}
 {bone1Name=dyn_hair_01, bone2Name=dyn_hair_side_front_parent, swing1LimitDegrees=45, swing2LimitDegrees=45, twistLimitDegrees=15}
 {bone1Name=dyn_hair_02, bone2Name=dyn_hair_01, swing1LimitDegrees=45, swing2LimitDegrees=45, twistLimitDegrees=15}
 {bone1Name=dyn_hair_03, bone2Name=dyn_hair_side_front_parent, swing1LimitDegrees=45, swing2LimitDegrees=45, twistLimitDegrees=15}
 {
  bone1Name=dyn_hair_side_mid_base
  bone2Name=head
  swing1LimitDegrees=45
  swing1Motion=Free
  swing2LimitDegrees=45
  swing2Motion=Free
  twistLimitDegrees=45
  twistMotion=Free
 }
 {bone1Name=dyn_hair_side_mid_parent, bone2Name=dyn_hair_side_mid_base, swing1LimitDegrees=45, swing2LimitDegrees=45, twistLimitDegrees=15}
 {bone1Name=dyn_hair_04, bone2Name=dyn_hair_side_mid_parent, swing1LimitDegrees=45, swing2LimitDegrees=45, twistLimitDegrees=15}
 {
  bone1Name=dyn_hair_front_base
  bone2Name=head
  swing1LimitDegrees=45
  swing1Motion=Free
  swing2LimitDegrees=45
  swing2Motion=Free
  twistLimitDegrees=45
  twistMotion=Free
 }
 {bone1Name=dyn_hair_front_parent, bone2Name=dyn_hair_front_base, swing1LimitDegrees=15, swing2LimitDegrees=15, twistLimitDegrees=3}
 {bone1Name=dyn_hair_18, bone2Name=dyn_hair_front_parent, swing1LimitDegrees=15, swing2LimitDegrees=15, twistLimitDegrees=3}
 {bone1Name=dyn_hair_19, bone2Name=dyn_hair_18, swing1LimitDegrees=15, swing2LimitDegrees=15, twistLimitDegrees=3}
 {bone1Name=dyn_hair_20, bone2Name=dyn_hair_front_parent, swing1LimitDegrees=15, swing2LimitDegrees=15, twistLimitDegrees=3}
 {bone1Name=dyn_hair_21, bone2Name=dyn_hair_20, swing1LimitDegrees=15, swing2LimitDegrees=15, twistLimitDegrees=3}
 {
  bone1Name=dyn_hair_front_strand_parent
  bone2Name=dyn_hair_front_strand_base
  swing1LimitDegrees=8
  swing2LimitDegrees=8
  twistLimitDegrees=3
 }
 {bone1Name=dyn_hair_front_strand_01, bone2Name=dyn_hair_front_strand_parent, swing1LimitDegrees=8, swing2LimitDegrees=8, twistLimitDegrees=3}
 {
  bone1Name=dyn_hair_front_strand_base
  bone2Name=head
  swing1LimitDegrees=45
  swing1Motion=Free
  swing2LimitDegrees=45
  swing2Motion=Free
  twistLimitDegrees=45
  twistMotion=Free
 }
 {
  bone1Name=dyn_hair_15_base
  bone2Name=head
  swing1LimitDegrees=45
  swing1Motion=Free
  swing2LimitDegrees=45
  swing2Motion=Free
  twistLimitDegrees=45
  twistMotion=Free
 }
 {bone1Name=dyn_hair_15, bone2Name=dyn_hair_15_base, swing1LimitDegrees=20, swing2LimitDegrees=20, twistLimitDegrees=10}
 {bone1Name=dyn_hair_16, bone2Name=dyn_hair_15, swing1LimitDegrees=1, swing2LimitDegrees=15, twistLimitDegrees=3}
 {bone1Name=dyn_hair_17, bone2Name=dyn_hair_16, swing1LimitDegrees=18, swing2LimitDegrees=18, twistLimitDegrees=3}
 {
  bone1Name=dyn_hair_back_base_06
  bone2Name=head
  swing1LimitDegrees=45
  swing1Motion=Free
  swing2LimitDegrees=45
  swing2Motion=Free
  twistLimitDegrees=45
  twistMotion=Free
 }
 {bone1Name=dyn_hair_06, bone2Name=dyn_hair_back_base_06, swing1LimitDegrees=20, swing2LimitDegrees=20, twistLimitDegrees=10}
 {bone1Name=dyn_hair_07, bone2Name=dyn_hair_06, swing1LimitDegrees=30, swing2LimitDegrees=30, twistLimitDegrees=10}
 {
  bone1Name=dyn_hair_back_base_08
  bone2Name=head
  swing1LimitDegrees=45
  swing1Motion=Free
  swing2LimitDegrees=45
  swing2Motion=Free
  twistLimitDegrees=45
  twistMotion=Free
 }
 {bone1Name=dyn_hair_08, bone2Name=dyn_hair_back_base_08, swing1LimitDegrees=30, swing2LimitDegrees=30, twistLimitDegrees=5}
 {bone1Name=dyn_hair_09, bone2Name=dyn_hair_08, swing1LimitDegrees=45, swing2LimitDegrees=45, twistLimitDegrees=5}
 {
  bone1Name=dyn_hair_back_base_10
  bone2Name=head
  swing1LimitDegrees=45
  swing1Motion=Free
  swing2LimitDegrees=45
  swing2Motion=Free
  twistLimitDegrees=45
  twistMotion=Free
 }
 {bone1Name=dyn_hair_10, bone2Name=dyn_hair_back_base_10, swing1LimitDegrees=20, swing2LimitDegrees=20, twistLimitDegrees=10}
 {bone1Name=dyn_hair_11, bone2Name=dyn_hair_10, swing1LimitDegrees=45, swing2LimitDegrees=30, twistLimitDegrees=10}
 {
  bone1Name=dyn_hair_back_base_12
  bone2Name=head
  swing1LimitDegrees=45
  swing1Motion=Free
  swing2LimitDegrees=45
  swing2Motion=Free
  twistLimitDegrees=45
  twistMotion=Free
 }
 {bone1Name=dyn_hair_12, bone2Name=dyn_hair_back_base_12, swing1LimitDegrees=20, swing2LimitDegrees=20, twistLimitDegrees=10}
 {bone1Name=dyn_hair_13, bone2Name=dyn_hair_12, swing1LimitDegrees=45, swing2LimitDegrees=30, twistLimitDegrees=10}
 {bone1Name=dyn_hair_14, bone2Name=dyn_hair_13, swing1LimitDegrees=45, swing2LimitDegrees=20, twistLimitDegrees=3}
 {
  bone1Name=dyn_hair_side_back_base
  bone2Name=head
  swing1LimitDegrees=45
  swing1Motion=Free
  swing2LimitDegrees=45
  swing2Motion=Free
  twistLimitDegrees=45
  twistMotion=Free
 }
 {bone1Name=dyn_hair_side_back_parent, bone2Name=dyn_hair_side_back_base, swing1LimitDegrees=20, swing2LimitDegrees=20, twistLimitDegrees=10}
 {bone1Name=dyn_hair_05, bone2Name=dyn_hair_side_back_parent, swing1LimitDegrees=45, swing2LimitDegrees=30, twistLimitDegrees=10}
]
